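# Material function that layers grime and then scratches over an input material, filed with the Ch6 pickup truck materials.
# The Grime mask, reshaped by spread and intensity, blends in tinted base color with adjusted roughness, specular, metallic and CustomData1.
# The Scratch mask blends that result toward a desaturated, tinted or overridden color with scaled surface attributes.
# It outputs the blended attributes and scratch color; the Gmap remap feeds nothing.
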MaterialFunction /Hermes/Environments/Ch6/Props/Vehicle/Pickup_Truck/Materials/MF_Valet_ScratchGrime
graph {nodes [
  n1 = Multiply(
   A=FunctionInput() {inputName=Scratch, inputType=FunctionInput_Scalar}
   B=LinearInterpolate(Alpha=ScalarParameter() {defaultValue=0.5, group=Grime and Scratch, parameterName=Scratch Intensity, sliderMax=1}) {constB=4}
  )
  Add(A=Subtract(), B=@n1)
  n2 = FunctionInput() {inputName=Material, inputType=FunctionInput_MaterialAttributes}
  n3 = GetMaterialAttributes(None=@n2) {attributeGetTypes [#1, #2]}
  n4 = VectorParameter() {group=Grime and Scratch, parameterName=Grime Tint, defaultValue=<rgba 0.01 0.0085 0.0063 1>}
  n5 = ScalarParameter() {defaultValue=1, group=Grime and Scratch, parameterName=Grime Roughness}
  n6 = ScalarParameter() {defaultValue=1, group=Grime and Scratch, parameterName=Grim Metallic Multiplier}
  n7 = SetMaterialAttributes!(
   MaterialAttributes=@n2
   BaseColor=LinearInterpolate(
    A=Multiply(A=@n3.BaseColor, B=@n4.RGB)
    B=@n4.RGB
    Alpha=ScalarParameter() {group=Grime and Scratch, parameterName=GrimeTintOverride, sliderMax=1}
   )
   Roughness=@n5
   Specular=ScalarParameter() {defaultValue=0.35, group=Grime and Scratch, parameterName=Grime Spec}
   CustomData1=@n5
   Metallic=Saturate(None=Multiply(A=@n3.Metallic, B=@n6))
  ) {attributeSetTypes [
    #2
    #3
    #4
    #5
    #1
   ]}
  n8 = ScalarParameter() {defaultValue=1, group=Grime and Scratch, parameterName=Grime Spread, sliderMax=1}
  n9 = Multiply(A=Subtract(A=FunctionInput() {inputName=Grime, inputType=FunctionInput_Scalar}, B=Subtract(B=@n8)), B=Divide(B=@n8))
  n10 = ScalarParameter() {defaultValue=0.5, group=Grime and Scratch, parameterName=Grime Intensity, sliderMax=1}
  n11 = BlendMaterialAttributes(A=@n2, B=@n7, Alpha=Saturate(None=Multiply(A=@n9, B=@n10)))
  n12 = GetMaterialAttributes(None=@n11) {attributeGetTypes [
    #2
    #3
    #5
    #1
    #4
   ]}
  n13 = Desaturation(None=@n12.BaseColor, Fraction=ScalarParameter() {defaultValue=0.4, group=Grime and Scratch, parameterName=Scratch Base Color Desaturation})
  n14 = VectorParameter() {group=Grime and Scratch, parameterName=Scratch Tint, defaultValue=<rgba 1.1335 1.1984 1.25 1>}
  n15 = ScalarParameter() {
   desc="0 : use scratchTint to tint the base color for the color of scratches\r\n1 : use scratchTint directly for color of scratches."
   group=Grime and Scratch
   parameterName=ScratchTintOverride
  }
  n16 = LinearInterpolate(A=Saturate(None=Multiply(A=@n13, B=@n14.RGB)), B=@n14.RGB, Alpha=@n15)
  n17 = ScalarParameter() {defaultValue=0.25, group=Grime and Scratch, parameterName=Scratch Roughness Multiplier}
  n18 = ScalarParameter() {defaultValue=1, group=Grime and Scratch, parameterName=Scratch Specular Multiplier}
  n19 = ScalarParameter() {defaultValue=1, group=Grime and Scratch, parameterName=Scratch Metalness Multiplier}
  n20 = SetMaterialAttributes!(
   MaterialAttributes=@n11
   BaseColor=@n16
   Roughness=Saturate(None=Multiply(A=@n12.Roughness, B=@n17))
   Specular=Saturate(None=Multiply(A=@n12.Specular, B=@n18))
   Metallic=Saturate(None=Multiply(A=@n12.Metallic, B=@n19))
   CustomData1=Saturate(None=Multiply(A=@n12.CustomData1, B=@n17))
  ) {attributeSetTypes [
    #2
    #3
    #4
    #1
    #5
   ]}
  FunctionOutput(None=BlendMaterialAttributes(A=@n11, B=@n20, Alpha=Saturate(None=@n1)))
  FunctionOutput(None=@n16) {outputName=Scratch Color}
  LinearInterpolate(Alpha=FunctionInput() {inputName=Gmap, inputType=FunctionInput_Scalar}) {constA=0.2, constB=1.12}
 ]}
