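# Material function in the Ch6 pickup truck prop folder that fades out the surface as the camera nears it, to prevent view occlusion.
# Camera distance to a local-space FadePosition parameter, transformed to world space, is remapped between EndFade and StartFade and saturated.
# The result multiplies the incoming material attributes' OpacityMask, so the fade keys off one point rather than each pixel's distance.
# The UseOcclusionPrevention static switch otherwise passes the attributes through unchanged.
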MaterialFunction /Hermes/Environments/Ch6/Props/Vehicle/Pickup_Truck/Materials/MF_OcclusionPrevention
graph {nodes [
  n1 = FunctionInput() {inputType=FunctionInput_MaterialAttributes}
  n2 = GetMaterialAttributes(None=@n1) {attributeGetTypes [#1]}
  n3 = ScalarParameter() {defaultValue=50, group=OcclusionPrevention, parameterName=EndFade}
  n4 = ScalarParameter() {defaultValue=70, group=OcclusionPrevention, parameterName=StartFade}
  n5 = Distance(
   A=CameraPositionWS()
   B=TransformPosition(None=VectorParameter() {group=OcclusionPrevention, parameterName="FadePosition(LocalToWorld)", defaultValue=<rgba 63 122 335 0>}.RGB) {transformType=TRANSFORMPOSSOURCE_World}
  )
  n6 = Multiply(A=@n2.OpacityMask, B=Saturate(None=InverseLinearInterpolate(A=@n3, B=@n4, Value=@n5)))
  FunctionOutput(None=StaticSwitchParameter(
    True=SetMaterialAttributes!(MaterialAttributes=@n2.MaterialAttributes, OpacityMask=@n6, EmissiveColor) {attributeSetTypes [#1, #2]}
    False=@n1
   ) {parameterName=UseOcclusionPrevention})
 ]}
