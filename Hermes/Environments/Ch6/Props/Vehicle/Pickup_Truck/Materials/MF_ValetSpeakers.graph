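# Material function for the pickup truck's valet speakers that adds animated world position offset behind a static switch.
# Centered UVs become a transformed direction vector, scaled by a default size plus an animated amount.
# Custom primitive data supplies a timed pulse's start time and a radio-on weight blending that pulse toward two multiplied ambient-music sine waves.
# The shadow pass keeps the input offset, and opacity mask comes from Diffuse alpha.
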MaterialFunction /Hermes/Environments/Ch6/Props/Vehicle/Pickup_Truck/Materials/MF_ValetSpeakers
graph {nodes [
  FunctionOutput(None=StaticSwitchParameter(True=NamedRerouteUsage(), False=NamedRerouteUsage()) {group=SpeakerWPO, parameterName=UseSpeakerWPO})
  n1 = GetMaterialAttributes(None=FunctionInput() {inputName=Material, inputType=FunctionInput_MaterialAttributes}) {attributeGetTypes [#1, #2]}
  NamedRerouteDeclaration(None=@n1.MaterialAttributes) {name=Input_Material, nodeColor=<rgba 0 1 0.6016 1>}
  NamedRerouteDeclaration(None=@n1.WorldPositionOffset) {name=Input_WPO, nodeColor=<rgba 1 0.2812 0 1>}
  n2 = NamedRerouteUsage()
  n3 = Multiply(A=TextureCoordinate()) {constB=2}
  n4 = Transform(None=AppendVector(A=Subtract(A=@n3), B=Constant()))
  n5 = ScalarParameter() {defaultValue=-4, group=SpeakerWPO, parameterName=SpeakerDefaultScale}
  n6 = Time()
  n7 = ScalarParameter() {bUseCustomPrimitiveData=true, group=SpeakerWPO, parameterName=W_SpeakerAnimation, primitiveDataIndex=1}
  n8 = Saturate(None=Divide(A=Subtract(A=@n6, B=@n7), B=ScalarParameter() {defaultValue=0.5, parameterName=Duration}))
  n9 = Time()
  n10 = ScalarParameter() {defaultValue=1, parameterName=AmbientMusicBaseFreq}
  n11 = Multiply(A=Sine(None=Multiply(A=@n9, B=@n10)), B=ScalarParameter() {defaultValue=5, parameterName=TempBase_Bounce})
  n12 = ScalarParameter() {defaultValue=2, parameterName=AmbientMusicStemFreq}
  n13 = Multiply(A=Sine(None=Multiply(A=@n9, B=@n12)), B=ScalarParameter() {defaultValue=2, parameterName=TempStem_Bounce})
  n14 = LinearInterpolate(
   A=Multiply(A=Sine(None=@n8), B=ScalarParameter() {defaultValue=5, group=SpeakerWPO, parameterName=Scale_Amt}) {constB=12}
   B=ConstantBiasScale(None=Multiply(A=@n11, B=@n13))
   Alpha=ScalarParameter() {bUseCustomPrimitiveData=true, parameterName=W_RadioOn, primitiveDataIndex=2}
  )
  n15 = Add(A=@n2, B=Add(A=Multiply(A=@n4, B=@n5) {constB=-2}, B=Multiply(A=@n4, B=@n14)))
  n16 = TextureSampleParameter2D() {
   group=Texture Input
   parameterName=Diffuse
   samplerSource=SSM_Wrap_WorldGroupSettings
   texture=&T_BlacksRGB
  }
  NamedRerouteDeclaration(None=SetMaterialAttributes!(
    MaterialAttributes=NamedRerouteUsage()
    WorldPositionOffset=ShadowReplace(Default=@n15, Shadow=@n2)
    OpacityMask=ComponentMask(None=@n16.RGBA) {a=true}
   ) {attributeSetTypes [#1, #3]}) {name=Outout_Material, nodeColor=<rgba 0 0.0391 1 1>}
 ]}
