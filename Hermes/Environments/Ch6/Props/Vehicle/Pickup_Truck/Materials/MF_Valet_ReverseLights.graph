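# Material function that adds controllable reverse-light glow to the Valet pickup truck by modifying the emissive channel of incoming material attributes.
# Glow color comes from desaturated, contrast-boosted base color tinted by a color parameter and scaled by ReverseLight.
# A TextureCropping UV rectangle and the mask texture's red channel, also scaled by ReverseLight, define where it applies.
# A lerp overwrites incoming emissive only in that region, replacing an Add node left unconnected.
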MaterialFunction /Hermes/Environments/Ch6/Props/Vehicle/Pickup_Truck/Materials/MF_Valet_ReverseLights
graph {nodes [
  n1 = GetMaterialAttributes(None=FunctionInput() {inputName=MA, inputType=FunctionInput_MaterialAttributes}) {attributeGetTypes [#1, #2]}
  n2 = Desaturation(None=@n1.BaseColor)
  n3 = ScalarParameter() {defaultValue=3, group=ReverseLights, parameterName=ReverseContrast}
  n4 = ScalarParameter() {group=ReverseLights, parameterName=ReverseLight}
  n5 = Multiply(
   A=Multiply(
    A=Power(Base=@n2, Exp=@n3)
    B=VectorParameter() {group=ReverseLights, parameterName=ReverseLight Color, defaultValue=<rgba 200 200 200 1>}.RGB
   )
   B=@n4
  )
  Add(A=@n1.EmissiveColor, B=@n5)
  n6 = MaterialFunctionCall!(
   UVs
   TextureIn
   UpperLeftCorner=AppendVector(
    A=ScalarParameter() {defaultValue=0.25, group=ReverseLights, parameterName=ReversLightUV_TopLeft_X}
    B=ScalarParameter() {defaultValue=0.625, group=ReverseLights, parameterName=ReversLightUV_TopLeft_Y}
   )
   LowerRightCorner=AppendVector(
    A=ScalarParameter() {defaultValue=0.5, group=ReverseLights, parameterName=ReversLightUV_BottomRight_X}
    B=ScalarParameter() {defaultValue=0.875, group=ReverseLights, parameterName=ReversLightUV_BottomRight_Y}
   )
   ExportFloat4
  ) {materialFunction=/Engine/Functions/Engine_MaterialFunctions01/Texturing/TextureCropping}
  n7 = TextureSampleParameter2D(UVs=@n6.Crop UVs) {
   group=Textures
   parameterName=ReverseLightMask
   samplerType=SAMPLERTYPE_Masks
   texture=/Hermes/Environments/Ch6/Props/Vehicle/Pickup_Truck/Textures/T_Valet_Sedan_E2
  }
  n8 = Multiply(A=Multiply(A=@n6.Crop Mask, B=@n7.R), B=@n4)
  FunctionOutput(None=SetMaterialAttributes!(
    MaterialAttributes=@n1.MaterialAttributes
    EmissiveColor=LinearInterpolate(A=@n1.EmissiveColor, B=@n5, Alpha=@n8) {desc=Swapped to Lerp instead of Add so we can overwrite emissive values passed in but only where the lights are}
   ) {attributeSetTypes [#1]})
 ]}
