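# Master material for Valet vehicles, stored with the pickup truck, using Default Lit or Clear Coat shading.
# Chained material functions layer base, headlights, turn signals, decal, reverse lights, damage, speakers, party lights and occlusion prevention.
# Pre-skinned position and bounds split lights and damage by side, though the divisor subtracts bounds Max from itself.
# Optional material layers feed the de-rez transition; the mud function's output is unused.
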Material /Hermes/Environments/Ch6/Props/Vehicle/Pickup_Truck/Materials/M_FN_Valet_Master
graph {
 properties {
  bUseMaterialAttributes=true
  bUsedWithMeshDeformer=true
  bUsedWithNanite=true
  bUsedWithSkeletalMesh=true
  ditherOpacityMask=true
  opacityMaskClipValue=0.5
  pixelDepthOffsetMode=PDOM_Legacy
  shadingModel=MSM_FromMaterialExpression
  usedShadingModels=Default Lit | Clear Coat
 }
 nodes [
  n1 = PreSkinnedPosition()
  n2 = PreSkinnedLocalBounds()
  n3 = Divide(A=Subtract(A=VertexInterpolator(VS=@n1).PS, B=@n2.Min), B=Subtract(A=@n2.Max, B=@n2.Max))
  n4 = Round(None=ComponentMask(None=@n3) {r=true})
  n5 = MaterialFunctionCall() {materialFunction=&/MF_Valet_BaseFunctionality}
  n6 = MaterialFunctionCall!(
   In=MaterialFunctionCall!(Mask=Multiply(A=@n4, B=@n5.Lights), MA=@n5.Result) {materialFunction=&/MF_Valet_Headlights}.Result
   CenterLine=StaticSwitchParameter(True=ComponentMask(None=@n3) {g=true}, False=Constant()) {defaultValue=true, parameterName=UseCenterLine}
   Alpha=Multiply(A=@n5.Lights, B=OneMinus(None=@n4))
  ) {materialFunction=/Game/Weapons/WeaponSkins/MaterialFunction/TurnSignals}
  n7 = MaterialFunctionCall!(
   In=MaterialFunctionCall!(MA=MaterialFunctionCall!(Material=@n6.Result) {materialFunction=&/MF_Valet_Decal}.Result) {materialFunction=&/MF_Valet_ReverseLights}.Result
   Metal Tint=@n5.Scratch Color
   Local Bounds Split=@n3
  ) {materialFunction=/Game/Vehicles/Valet/MF_Valet_Damage}
  n8 = MaterialFunctionCall!(In=MaterialFunctionCall!(In=MaterialFunctionCall!(Material=@n7.Result) {materialFunction=&/MF_ValetSpeakers}.Result) {materialFunction=/Game/Packages/Fortress_SharedMaterials/PartyBus_TempLocation/MF_ArmoredBattleBus_PartyLights}.Result) {materialFunction=&/MF_OcclusionPrevention}
  MaterialFunctionCall(In=StaticSwitchParameter(True=MaterialAttributeLayers!(None=@n8.Result), False=@n8.Result) {parameterName=Use Material Layers}) {materialFunction=&MF_VehicleDeRez_Transition}
  MaterialFunctionCall!(Material=@n7.Result, PreskinnedPos=@n1) {materialFunction=&/MF_ValetMud}
 ]
}
